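# Niagara death effect for the AnnoyedMop heavy-attack projectile, bursting blue-purple lingering fire and embers along its segments.
# Particles are partitioned across NumSegments, placed in cylinders built from user position and rotation arrays, and killed where ActiveArray marks a segment inactive.
# Linger_Fire plays random-start SubUV flames, while GPU SegmentEmbers add curl-noise, speed-stretched sparks, both with vertically flattened cone velocity.
# The FireEject emitter is disabled.
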NiagaraSystem /LemonExam/Assets/FX/HeavyAttack/AnnoyedMop/NS_AnnoyedMop_Prj_OnDeath
dependencies {
 usedDataInterfaces [
  &NiagaraDataInterfaceCurve
  &NiagaraDataInterfaceArrayFloat
  &NiagaraDataInterfaceCamera
  &NiagaraDataInterfacePlatformSet
 ]
 usedDynamicInputs [
  &RemapRange
  &self:GetArrayLength
  &MakeCustomFloatFromBool
  &FN_PickFloatDependOnPlatform
 ]
 usedModules [
  &self:SystemSpawnScript.SystemScriptSource.SystemScriptGraph.NiagaraNodeAssignment_2.SetVariables_#1
  &self:SystemSpawnScript.SystemScriptSource.SystemScriptGraph.NiagaraNodeAssignment_3.SetVariables_#2
  &self:SystemSpawnScript.SystemScriptSource.SystemScriptGraph.NiagaraNodeAssignment_4.SetVariables_#3
  &self:SystemSpawnScript.SystemScriptSource.SystemScriptGraph.NiagaraNodeAssignment_5.SetVariables_#4
  &self:SystemSpawnScript.SystemScriptSource.SystemScriptGraph.NiagaraNodeAssignment_6.SetVariables_#5
  &self:SystemSpawnScript.SystemScriptSource.SystemScriptGraph.NiagaraNodeAssignment_7.SetVariables_#6
  &self:SystemSpawnScript.SystemScriptSource.SystemScriptGraph.NiagaraNodeAssignment_10.SetVariables_#7
  &self:SystemSpawnScript.SystemScriptSource.SystemScriptGraph.NiagaraNodeAssignment_11.SetVariables_#8
  &self:SystemSpawnScript.SystemScriptSource.SystemScriptGraph.NiagaraNodeAssignment_9.SetVariables_#9
  &self:SystemSpawnScript.SystemScriptSource.SystemScriptGraph.NiagaraNodeAssignment_8.SetVariables_#10
  &self:SystemSpawnScript.SystemScriptSource.SystemScriptGraph.NiagaraNodeAssignment_13.SetVariables_#11
  &self:SystemSpawnScript.SystemScriptSource.SystemScriptGraph.NiagaraNodeAssignment_14.SetVariables_#12
  &SystemState
  &self:SystemSpawnScript.SystemScriptSource.SystemScriptGraph.NiagaraNodeAssignment_0.SetVariables_#13
  &self:SystemSpawnScript.SystemScriptSource.SystemScriptGraph.NiagaraNodeAssignment_1.SetVariables_#14
  &Timeline
  &self:SystemSpawnScript.SystemScriptSource.SystemScriptGraph.NiagaraNodeAssignment_12.SetVariables_#15
 ]
}
emitters {
 FireEject {
  bEnabled=false
  emitterUpdateScript [
   &EmitterState(Life Cycle Mode=::NewEnumerator1, Loop Behavior=::NewEnumerator1, Loop Duration=0.6, UseLoopDelay=false)
   &SpawnBurst_Instantaneous(Spawn Count=10, Use Spawn Probability=true, UseLoopCountLimit=false) {module=SpawnBurst_Instantaneous002}
  ]
  particleSpawnScript [
   &InitializeParticle(
    Color=<rgba 20 2.3306 0 4>
    Lifetime=&RandomRangeFloat(Minimum=0.3, Maximum=0.5)
    Position Offset=&Multiply_VectorByFloat(Vector=<0 0 0>, Float=1)
    Sprite Rotation Angle=&RandomRangeFloat(Minimum=0, Maximum=1)
    Sprite Rotation Mode=::NewEnumerator1
    Sprite Size=&Multiply_Vector2D_ByFloat(
     Float=&Multiply_Float(A=&RandomRangeFloat(Minimum=90, Maximum=180), B=0.5)
     Vector2D=<1 3>
    )
    Sprite Size Mode=::NewEnumerator4
    Sprite UV Mode=::NewEnumerator3
    UsePositionOffset=true
    Write Material Random=true
   ) {module=InitializeParticle001}
   &ShapeLocation(
    Handle Radius=0
    Large Radius=&Multiply_Float(A=10, B=1)
    MaskSpawnGroup=false
    Shape Primitive=::NewEnumerator3
    Torus Distribution Mode:ENiagaraTorusDistributionMode=::NewEnumerator1
    U Position=&ReturnNormalizedExecIndex(Normalized Index Includes One=false)
    UseOwnerScale=true
   ) {enabled=false, module=ShapeLocation001}
   &PartitionParticles(Number Of Partitions=$User.NumSegments)
   /SetVariables(Particles.OriginPos:NiagaraPosition=&TransformPosition(
     Position=&Multiply_VectorByFloat(
      Vector=&Multiply_Vector(
       A=&SelectVectorFromArray(Direct Array Index=$Particles.PartitionParticles.Partition, Vector Selection Array=$User.SegmentPositions)
       B=$User.RootScale
      )
      Float=1
     )
     Source Space=::Local
     Destination Space=::Local
    )) {isSetParameters=true, module=SetVariables_#16}
   &ShapeLocation(
    Custom Transform Matrix:NiagaraMatrix=/Niagara/DynamicInputs/Transforms/MultiplyMatrix(
     Matrix A:NiagaraMatrix=/Niagara/DynamicInputs/Math/MakeTransform(
      Inverse:NiagaraBool=false
      Position:Vector3f=?
      Local Offset:Vector3f=<0 0 0>
      Scale:Vector3f=<1 1 1>
      Rotation Interface:ENiagara_RotationDefinitionApproach=::NewEnumerator1
      Axis:Vector3f=?
      Quaternion:Quat4f=&MultiplyQuaternion(
       Quaternion A=&self:ScratchDynamicInput_02(Index:NiagaraInt32=$Particles.PartitionParticles.Partition, QuatArray:NiagaraDataInterfaceArrayQuat=$User.SegmentRotations)
       Quaternion B=&MakeQuatFromAxisAngle(Axis=<1 0 0>, AngleInDegrees=90)
      )
      Rotation Matrix:NiagaraMatrix=?
      X Vector:Vector3f=?
      Yaw:NiagaraFloat=?
      Angle:NiagaraFloat=?
      Pitch:NiagaraFloat=?
      Y Vector:Vector3f=?
      Roll:NiagaraFloat=?
      Z Vector:Vector3f=?
     )
     Matrix B:NiagaraMatrix=&MakeTransform(
      Inverse:NiagaraBool=false
      Position:Vector3f=Linked() {linkedVariable {name=Particles.OriginPos, type=NiagaraPosition}}
      Local Offset:Vector3f=<0 0 0>
      Scale:Vector3f=<1 1 1>
      Rotation Interface:ENiagara_RotationDefinitionApproach=::NewEnumerator2
      Axis:Vector3f=?
      Quaternion:Quat4f=?
      Rotation Matrix:NiagaraMatrix=?
      X Vector:Vector3f=?
      Yaw:NiagaraFloat=0
      Angle:NiagaraFloat=?
      Pitch:NiagaraFloat=0
      Y Vector:Vector3f=?
      Roll:NiagaraFloat=0
      Z Vector:Vector3f=?
     )
    )
    Cylinder Height=130
    Cylinder Radius=1
    MaskSpawnGroup=false
    Override Seed=false
    Shape Primitive=::NewEnumerator1
    Surface Only=false
    Transform Method:ENiagara_TransformType=::NewEnumerator1
   ) {module=ShapeLocation002}
   &AddVelocity(Velocity=$Particles.ShapeLocation.ShapeNormal, Velocity Speed Scale=100)
   &CameraOffset(Camera Offset Amount=14) {module=CameraOffset001}
   /SetVariables(
    Particles.Scale=&Multiply_VectorByFloat(Vector=$Particles.Scale, Float=&Multiply_Float(A=1, B=1))
    Particles.SpriteSize=&Multiply_Vector2D_ByFloat(Float=1, Vector2D=$Particles.SpriteSize)
   ) {enabled=false, isSetParameters=true, module=SetVariables_#17}
   &AddVelocity(
    Cone Angle=65
    Module.UseConeFalloff=false
    Module.UseVelDistribution=false
    Override Seed=false
    Velocity Mode=::NewEnumerator2
    Velocity Speed=&RandomRangeFloat(Minimum=2000, Maximum=3500)
   ) {module=AddVelocity002}
   &KillParticles(Kill Particles=&InvertBool(Boolean:NiagaraBool=&self:ScratchDynamicInput(SegmentIndex:NiagaraInt32=$Particles.PartitionParticles.Partition, VisibilityArray:NiagaraDataInterfaceArrayBool=$User.ActiveArray)))
   /SetVariables(Particles.Velocity=&Multiply_Vector(A=$Particles.Velocity, B=<1 1 0.1>)) {isSetParameters=true, module=SetVariables_#18}
  ]
  particleUpdateScript [
   &ParticleState()
   &Drag(Drag=&RandomRangeFloat(Minimum=2, Maximum=3), Use Linear Drag=true, Use Rotational Drag=false)
   &AccelerationForce() {module=AccelerationForce001}
   &SpriteRotationRate(Rotation Rate=&Multiply_Float(
     A=&UniformRangedFloat(Minimum=1, Maximum=1)
     B=&FloatFromCurve(
      FloatCurve=DataInterface() {propertyValues {Curve {keys [{arriveTangent=-3.245, interpMode=RCIM_Cubic, leaveTangent=-3.245, tangentMode=RCTM_User, time=0, value=1}, <at 0.1 0.06>, <at 1.0011 0.5>]}}}
      Scale Curve=&SelectFloatFromArray(
       Array Sampling Mode=::NewEnumerator0
       Direct Array Index=?
       Float Selection Array=DataInterface() {propertyValues {FloatData [-100,100]}}
       Random Seed=0
      )
     )
    )) {enabled=false}
   &self:ViewpointBased_Sprite_Rotation_03() {enabled=false, module=ViewpointBased_Sprite_Rotation_03}
   /Niagara/Modules/Update/Forces/CurlNoiseForce(
    Noise Frequency:NiagaraFloat=&Divide_Float(A=1, B=1)
    Noise Quality / Cost=::NewEnumerator0
    Noise Strength=&Multiply_Float(A=222, B=1)
    Pan Noise=true
    Pan Noise Field=<0 0 0.1>
   ) {enabled=false}
   &SolveForcesAndVelocity(Clamp Velocity=false, Limit Acceleration=false)
   &ScaleSpriteSize(Scale Factor=&Vector2DFromFloat(Value=&FloatFromCurve(FloatCurve=DataInterface() {propertyValues {Curve {keys [
          {arriveTangent=6.4923, interpMode=RCIM_Cubic, leaveTangent=6.4923, tangentMode=RCTM_User, time=0, value=0.5}
          {interpMode=RCIM_Cubic, leaveTangent=0.155, time=1, value=2}
         ]}}})))
   &ScaleColor(
    Scale Alpha=&FloatFromCurve(FloatCurve=DataInterface() {propertyValues {Curve {keys [{arriveTangent=6.9262, interpMode=RCIM_Cubic, time=0.01, value=0.5}, {interpMode=RCIM_Cubic, leaveTangent=-0.6884, time=1.0011, value=0}]}}})
    Scale RGB=&Lerp_Vector(
     A=<0.15 0.15 0.15>
     B=<0.3 0.3 0.3>
     Alpha=&FloatFromCurve(FloatCurve=DataInterface() {propertyValues {Curve {keys [{tangentMode=RCTM_User, time=0.1, value=10}, {tangentMode=RCTM_User, time=0.35, value=1}]}}})
    )
    ScaleA=true
    ScaleRGB=false
   ) {module=ScaleColor001}
   &DynamicMaterialParameters(
    Index 0 Param 1=&FloatFromCurve(
     FloatCurve=DataInterface() {propertyValues {Curve {keys [{interpMode=RCIM_Cubic, tangentMode=RCTM_User, time=0, value=0}, {interpMode=RCIM_Cubic, tangentMode=RCTM_User, time=1, value=1}]}}}
     Scale Curve=5
    )
    Index 0 Param 4=1
    Module.Index0Param0Write=true
    Module.Index0Param1Write=true
    Module.Index0Param2Write=true
    Module.Index0Param3Write=false
   ) {module=DynamicMaterialParameters001}
  ]
  rendererClasses [&NiagaraSpriteRendererProperties, &NiagaraMeshRendererProperties]
  renderers [
   {rendererClass=&NiagaraSpriteRendererProperties}
   {rendererClass=&NiagaraMeshRendererProperties, rendererIndex=1}
  ]
 }
 Linger_Fire {
  emitterUpdateScript [
   &EmitterState(Life Cycle Mode=::NewEnumerator1, Loop Behavior=::NewEnumerator1, Loop Duration=0.6, UseLoopDelay=false)
   &SpawnBurst_Instantaneous(Spawn Count=&Multiply_Int(A=4, B=$User.NumSegments), Use Spawn Probability=true, UseLoopCountLimit=false) {module=SpawnBurst_Instantaneous002}
  ]
  particleSpawnScript [
   &InitializeParticle(
    Color Maximum=<rgba 7.7113 1.7478 15 1>
    Color Minimum=<rgba 1.5148 1.8976 13 1>
    Color Mode=::NewEnumerator2
    Lifetime=&RandomRangeFloat(Minimum=0.3, Maximum=0.75)
    Position Offset=&Multiply_VectorByFloat(Vector=<0 0 0>, Float=1)
    Sprite Rotation Angle=&RandomRangeFloat(Minimum=0, Maximum=1)
    Sprite Rotation Mode=::NewEnumerator1
    Sprite Size=&Multiply_Vector2D_ByFloat(
     Float=&Multiply_Float(A=&RandomRangeFloat(Minimum=90, Maximum=180), B=2.3662)
     Vector2D=<1 1>
    )
    Sprite Size Mode=::NewEnumerator4
    Sprite UV Mode=::NewEnumerator3
    UsePositionOffset=true
    Write Material Random=true
   ) {module=InitializeParticle001}
   &ShapeLocation(
    Handle Radius=0
    Large Radius=&Multiply_Float(A=10, B=1)
    MaskSpawnGroup=false
    Shape Primitive=::NewEnumerator3
    Torus Distribution Mode:ENiagaraTorusDistributionMode=::NewEnumerator1
    U Position=&ReturnNormalizedExecIndex(Normalized Index Includes One=false)
    UseOwnerScale=true
   ) {enabled=false, module=ShapeLocation001}
   &PartitionParticles(Number Of Partitions=$User.NumSegments)
   /SetVariables(Particles.OriginPos:NiagaraPosition=&TransformPosition(
     Position=&Multiply_VectorByFloat(
      Vector=&Multiply_Vector(
       A=&SelectVectorFromArray(Direct Array Index=$Particles.PartitionParticles.Partition, Vector Selection Array=$User.SegmentPositions)
       B=$User.RootScale
      )
      Float=1
     )
     Source Space=::Local
     Destination Space=::Local
    )) {isSetParameters=true, module=SetVariables_#16}
   &ShapeLocation(
    Custom Transform Matrix:NiagaraMatrix=&MultiplyMatrix(
     Matrix A:NiagaraMatrix=&MakeTransform(
      Inverse:NiagaraBool=false
      Position:Vector3f=?
      Local Offset:Vector3f=<0 0 0>
      Scale:Vector3f=<1 1 1>
      Rotation Interface:ENiagara_RotationDefinitionApproach=::NewEnumerator1
      Axis:Vector3f=?
      Quaternion:Quat4f=&MultiplyQuaternion(
       Quaternion A=&self:ScratchDynamicInput_02(Index:NiagaraInt32=$Particles.PartitionParticles.Partition, QuatArray:NiagaraDataInterfaceArrayQuat=$User.SegmentRotations)
       Quaternion B=&MakeQuatFromAxisAngle(Axis=<1 0 0>, AngleInDegrees=90)
      )
      Rotation Matrix:NiagaraMatrix=?
      X Vector:Vector3f=?
      Yaw:NiagaraFloat=?
      Angle:NiagaraFloat=?
      Pitch:NiagaraFloat=?
      Y Vector:Vector3f=?
      Roll:NiagaraFloat=?
      Z Vector:Vector3f=?
     )
     Matrix B:NiagaraMatrix=&MakeTransform(
      Inverse:NiagaraBool=false
      Position:Vector3f=Linked() {linkedVariable {name=Particles.OriginPos, type=NiagaraPosition}}
      Local Offset:Vector3f=<0 0 0>
      Scale:Vector3f=<1 1 1>
      Rotation Interface:ENiagara_RotationDefinitionApproach=::NewEnumerator2
      Axis:Vector3f=?
      Quaternion:Quat4f=?
      Rotation Matrix:NiagaraMatrix=?
      X Vector:Vector3f=?
      Yaw:NiagaraFloat=0
      Angle:NiagaraFloat=?
      Pitch:NiagaraFloat=0
      Y Vector:Vector3f=?
      Roll:NiagaraFloat=0
      Z Vector:Vector3f=?
     )
    )
    Cylinder Height=130
    Cylinder Radius=1
    MaskSpawnGroup=false
    Override Seed=false
    Shape Primitive=::NewEnumerator1
    Surface Only=false
    Transform Method:ENiagara_TransformType=::NewEnumerator1
   ) {module=ShapeLocation002}
   &AddVelocity(Velocity=$Particles.ShapeLocation.ShapeNormal, Velocity Speed Scale=100)
   &CameraOffset(Camera Offset Amount=14) {module=CameraOffset001}
   /SetVariables(
    Particles.Scale=&Multiply_VectorByFloat(Vector=$Particles.Scale, Float=&Multiply_Float(A=1, B=1))
    Particles.SpriteSize=&Multiply_Vector2D_ByFloat(Float=1, Vector2D=$Particles.SpriteSize)
   ) {enabled=false, isSetParameters=true, module=SetVariables_#17}
   &AddVelocity(
    Cone Angle=65
    Module.UseConeFalloff=false
    Module.UseVelDistribution=false
    Override Seed=false
    Velocity Mode=::NewEnumerator2
    Velocity Speed=&RandomRangeFloat(Minimum=1000, Maximum=2500)
   ) {module=AddVelocity002}
   &KillParticles(Kill Particles=&InvertBool(Boolean:NiagaraBool=&self:ScratchDynamicInput(SegmentIndex:NiagaraInt32=$Particles.PartitionParticles.Partition, VisibilityArray:NiagaraDataInterfaceArrayBool=$User.ActiveArray)))
   /SetVariables(Particles.Velocity=&Multiply_Vector(A=$Particles.Velocity, B=<1 1 0.1>)) {isSetParameters=true, module=SetVariables_#18}
   /SetVariables(Particles.RandomFrame=&RandomRangeInt(Minimum=0, Maximum=63)) {isSetParameters=true}
  ]
  particleUpdateScript [
   &ParticleState()
   &Drag(Drag=&RandomRangeFloat(Minimum=3.5, Maximum=5), Use Linear Drag=true, Use Rotational Drag=false)
   &AccelerationForce() {module=AccelerationForce001}
   &SpriteRotationRate(Rotation Rate=&Multiply_Float(
     A=&UniformRangedFloat(Minimum=1, Maximum=1)
     B=&FloatFromCurve(
      FloatCurve=DataInterface() {propertyValues {Curve {keys [{arriveTangent=-3.245, interpMode=RCIM_Cubic, leaveTangent=-3.245, tangentMode=RCTM_User, time=0, value=1}, <at 0.1 0.06>, <at 1.0011 0.5>]}}}
      Scale Curve=&SelectFloatFromArray(
       Array Sampling Mode=::NewEnumerator0
       Direct Array Index=?
       Float Selection Array=DataInterface() {propertyValues {FloatData [-100,100]}}
       Random Seed=0
      )
     )
    )) {enabled=false}
   &self:ViewpointBased_Sprite_Rotation_03() {enabled=false, module=ViewpointBased_Sprite_Rotation_03}
   &CurlNoiseForce(
    Noise Frequency:NiagaraFloat=&Divide_Float(A=1, B=1)
    Noise Quality / Cost=::NewEnumerator0
    Noise Strength=&Multiply_Float(A=222, B=1)
    Pan Noise=true
    Pan Noise Field=<0 0 0.1>
   ) {enabled=false}
   &SolveForcesAndVelocity(Clamp Velocity=false, Limit Acceleration=false)
   &ScaleSpriteSize(Scale Factor=&Vector2DFromFloat(Value=&FloatFromCurve(FloatCurve=DataInterface() {propertyValues {Curve {keys [{arriveTangent=6.4923, interpMode=RCIM_Cubic, time=0, value=0.5}, {interpMode=RCIM_Cubic, leaveTangent=0.155, time=1, value=2}]}}})))
   &ScaleColor(
    Scale Alpha=&FloatFromCurve(
     FloatCurve=DataInterface() {propertyValues {Curve {keys [{arriveTangent=6.9262, interpMode=RCIM_Cubic, time=0.01, value=0.5}, {interpMode=RCIM_Cubic, leaveTangent=-0.6884, time=1.0011, value=0}]}}}
     Scale Curve=2
    )
    Scale RGB=&Lerp_Vector(
     A=<0.15 0.15 0.15>
     B=<0.3 0.3 0.3>
     Alpha=&FloatFromCurve(FloatCurve=DataInterface() {propertyValues {Curve {keys [{tangentMode=RCTM_User, time=0.1, value=1}, {tangentMode=RCTM_User, time=1, value=0}]}}})
    )
    ScaleA=true
    ScaleRGB=true
   ) {module=ScaleColor001}
   &DynamicMaterialParameters(
    Index 0 Param 1=&Add_Float(
     A=&RandomRangeFloat(Minimum=0, Maximum=10)
     B=&FloatFromCurve(
      FloatCurve=DataInterface() {propertyValues {Curve {keys [
          {arriveTangent=-2.2561, interpMode=RCIM_Cubic, leaveTangent=-2.2561, tangentMode=RCTM_User, time=0, value=1}
          {arriveTangent=-1.0036, interpMode=RCIM_Cubic, leaveTangent=-1.0036, tangentMode=RCTM_User, time=0.2682, value=0.3}
          <at 0.8 0>
         ]}}}
      Scale Curve=5
     )
    )
    Index 0 Param 2=&FloatFromCurve(
     FloatCurve=DataInterface() {propertyValues {Curve {keys [
         {arriveTangent=0.0861, interpMode=RCIM_Cubic, leaveTangent=0.0861, tangentMode=RCTM_User, time=0.1, value=0.3}
         {arriveTangent=1.536, interpMode=RCIM_Cubic, leaveTangent=1.536, tangentMode=RCTM_User, time=1, value=1.05}
        ]}}}
     Scale Curve=0.9318
    )
    Index 0 Param 3=&FloatFromCurve(FloatCurve=DataInterface() {propertyValues {Curve {keys [{tangentMode=RCTM_User, time=0, value=0.125}, {tangentMode=RCTM_User, time=1, value=1}]}}})
    Index 0 Param 4=1
    Index 1 Param 1=&MaskFloatBySpawnGroup(
     Masked Float=&FloatFromCurve(FloatCurve=DataInterface() {propertyValues {Curve {keys [
          {arriveTangent=8.5612, interpMode=RCIM_Cubic, leaveTangent=8.5612, tangentMode=RCTM_User, time=0, value=-0.5}
          {arriveTangent=0.9995, interpMode=RCIM_Cubic, leaveTangent=0.9995, tangentMode=RCTM_User, time=0.25, value=0.85}
          {arriveTangent=-0.0405, interpMode=RCIM_Cubic, leaveTangent=-0.0405, tangentMode=RCTM_User, time=0.5, value=1}
         ]}}})
     Passthrough Float=&FloatFromCurve(
      FloatCurve=DataInterface() {propertyValues {Curve {keys [
          {arriveTangent=4.1785, interpMode=RCIM_Cubic, leaveTangent=4.1785, tangentMode=RCTM_User, time=0, value=-0.5}
          {arriveTangent=0.9995, interpMode=RCIM_Cubic, leaveTangent=0.9995, tangentMode=RCTM_User, time=0.3, value=0.8}
          {arriveTangent=-0.0405, interpMode=RCIM_Cubic, leaveTangent=-0.0405, tangentMode=RCTM_User, time=0.8, value=1}
         ]}}}
      Scale Curve=2
     )
     Spawn Group Mask=1
    )
    Index 1 Param 2=&Add_Float(
     A=&FloatFromCurve(
      FloatCurve=DataInterface() {propertyValues {Curve {keys [
          {arriveTangent=-12.0154, interpMode=RCIM_Cubic, leaveTangent=-12.0154, tangentMode=RCTM_User, time=0, value=1}
          {interpMode=RCIM_Cubic, tangentMode=RCTM_User, time=0.2, value=0}
         ]}}}
      Scale Curve=-1
     )
     B=&RandomRangeFloat(Minimum=-100, Maximum=100)
    )
    Index 1 Param 3=&FloatFromCurve(
     FloatCurve=DataInterface() {propertyValues {Curve {keys [{interpMode=RCIM_Cubic, tangentMode=RCTM_User, time=0.333, value=1.2}, {interpMode=RCIM_Cubic, tangentMode=RCTM_User, time=1, value=2}]}}}
     Scale Curve=0.201
    )
    Index 1 Param 4=&FloatFromCurve(FloatCurve=DataInterface() {propertyValues {Curve {keys [<at 0 0.6>,<at 1 0.3>]}}}, Scale Curve=0.5648)
    Index 2 Param 1=&FloatFromCurve(FloatCurve=DataInterface() {propertyValues {Curve {keys [
         {arriveTangent=0.0048, interpMode=RCIM_Cubic, leaveTangent=0.0048, tangentMode=RCTM_User, time=0, value=0.7}
         {interpMode=RCIM_Cubic, leaveTangent=1.688, time=0.75, value=1.5}
        ]}}})
    Index 2 Param 2=&Add_Float(
     A=&RandomRangeFloat(Minimum=0, Maximum=100)
     B=&FloatFromCurve(
      FloatCurve=DataInterface() {propertyValues {Curve {keys [
          {arriveTangent=0.5501, interpMode=RCIM_Cubic, leaveTangent=0.5501, tangentMode=RCTM_User, time=0, value=0}
          {arriveTangent=1.1924, interpMode=RCIM_Cubic, leaveTangent=1.1924, tangentMode=RCTM_User, time=1, value=0.8}
         ]}}}
      Scale Curve=&RandomRangeFloat(Minimum=-0.3, Maximum=-0.56)
     )
    )
    Index 2 Param 3=&FloatFromCurve(
     FloatCurve=DataInterface() {propertyValues {Curve {keys [{tangentMode=RCTM_User, time=0, value=0}, {tangentMode=RCTM_User, time=1, value=1}]}}}
     Scale Curve=3
    )
    Index 2 Param 4=&FloatFromCurve(
     FloatCurve=DataInterface() {propertyValues {Curve {keys [{arriveTangent=-0.4045, time=0, value=0.75}, {leaveTangent=-0.318, time=1, value=1}]}}}
     Scale Curve=0.75
    )
    Index 3 Param 1=&FloatFromCurve(
     FloatCurve=DataInterface() {propertyValues {Curve {keys [
         {arriveTangent=0.6183, interpMode=RCIM_Cubic, leaveTangent=0.6183, tangentMode=RCTM_User, time=0.1, value=0.02}
         {interpMode=RCIM_Cubic, tangentMode=RCTM_User, time=0.5, value=1}
        ]}}}
     Scale Curve=5
    )
    Index 3 Param 2=&FloatFromCurve(
     FloatCurve=DataInterface() {propertyValues {Curve {keys [<at 0 1>,<at 1 0>]}}}
     Scale Curve=&RandomRangeFloat(Minimum=0.5, Maximum=1)
    )
    Index 3 Param 3=&FloatFromCurve(
     FloatCurve=DataInterface() {propertyValues {Curve {keys [
         {arriveTangent=0.9433, interpMode=RCIM_Cubic, leaveTangent=0.9433, tangentMode=RCTM_User, time=0, value=0.6}
         {arriveTangent=0.3564, interpMode=RCIM_Cubic, leaveTangent=0.3564, tangentMode=RCTM_User, time=1, value=0.9}
        ]}}}
     Scale Curve=0.0005
    )
    Index 3 Param 4=1
    Module.Index0Param0Write=true
    Module.Index0Param1Write=true
    Module.Index0Param2Write=true
    Module.Index0Param3Write=false
    Module.Index1Param0Write=false
    Module.Index1Param1Write=false
    Module.Index1Param2Write=true
    Module.Index1Param3Write=true
    Module.Index2Param0Write=true
    Module.Index2Param1Write=true
    Module.Index2Param2Write=true
    Module.Index2Param3Write=true
    Module.Index3Param0Write=true
    Module.Index3Param1Write=true
    Module.Index3Param2Write=true
    Module.Index3Param3Write=true
    Write Parameter Index 1=true
    Write Parameter Index 2=true
    Write Parameter Index 3=true
   ) {enabled=false, module=DynamicMaterialParameters001}
   &FN_SubUVAnimation(End Frame=&Add_Integer(A=$Particles.RandomFrame, B=33), Start Frame=$Particles.RandomFrame)
   &DynamicMaterialParameters(
    Index 0 Param 1=&FloatFromCurve(
     FloatCurve=DataInterface() {propertyValues {Curve {keys [
         {arriveTangent=2.8718, interpMode=RCIM_Cubic, leaveTangent=2.8718, tangentMode=RCTM_User, time=0, value=0}
         {arriveTangent=-0.3942, interpMode=RCIM_Cubic, leaveTangent=-0.3942, tangentMode=RCTM_User, time=1, value=1}
        ]}}}
     Scale Curve=0.1
    )
    Module.Index0Param0Write=true
    Module.Index0Param1Write=true
    Module.Index0Param2Write=true
    Module.Index0Param3Write=true
   )
   /SetVariables(
    Particles.EjectedFireColor:LinearColor=&MultiplyLinearColor(
     Linear Color=<rgba 6.4726 2.5774 15.6 4>
     Scale Factor=&MakeVector4(
      X=1
      Y=1
      Z=1
      W=&FloatFromCurve(
       FloatCurve=DataInterface() {propertyValues {Curve {keys [{arriveTangent=6.9262, interpMode=RCIM_Cubic, time=0.01, value=0.5}, {interpMode=RCIM_Cubic, leaveTangent=-0.6884, time=0.75, value=0}]}}}
       Scale Curve=0.1766
      )
     )
    )
    Particles.EjectedFireParams:Vector4f=&MakeVector4(
     X=&FloatFromCurve(
      FloatCurve=DataInterface() {propertyValues {Curve {keys [{interpMode=RCIM_Cubic, tangentMode=RCTM_User, time=0, value=0}, {interpMode=RCIM_Cubic, tangentMode=RCTM_User, time=1, value=1}]}}}
      Scale Curve=5
     )
     Y=0
     Z=0
     W=0
    )
    Particles.EjectedFireSize:Vector2f=&Multiply_Vector2D(
     A=&Multiply_Vector2D_ByFloat(
      Float=&Multiply_Float(A=&RandomRangeFloat(Minimum=90, Maximum=180), B=0.5)
      Vector2D=<1 3>
     )
     B=&Vector2DFromFloat(Value=&FloatFromCurve(
       FloatCurve=DataInterface() {propertyValues {Curve {keys [
           {arriveTangent=6.4923, interpMode=RCIM_Cubic, leaveTangent=6.4923, tangentMode=RCTM_User, time=0, value=0.5}
           {interpMode=RCIM_Cubic, leaveTangent=0.155, time=1, value=2}
          ]}}}
       Scale Curve=1.7814
      ))
    )
   ) {isSetParameters=true}
   &FN_Sprite_Attributes(
    Color=&Multiply_LinearColorByFloat(
     Linear Color=<rgba 0.0696 0 1 0.114>
     Float=&FloatFromCurve(FloatCurve=DataInterface() {propertyValues {Curve {
         defaultValue=1
         keys [
          {arriveTangent=9.0725, interpMode=RCIM_Cubic, leaveTangent=9.0725, tangentMode=RCTM_User, time=0, value=0}
          {arriveTangent=-0.6422, interpMode=RCIM_Cubic, leaveTangent=-0.6422, tangentMode=RCTM_User, time=0.1408, value=0.9847}
          {arriveTangent=-1.0994, interpMode=RCIM_Cubic, leaveTangent=-1.0994, tangentMode=RCTM_User, time=1, value=0}
         ]
        }}})
    )
    Create Color Attribute=true
    Create SpriteSize Attribute=true
    SpriteSize=&Multiply_Vector2D(A=$Particles.SpriteSize, B=<1 1>)
    Write to Module Namespace=true
   )
   /SetVariables(Particles.Vector 4:Vector4f=<1 135.6608 1 1>) {isSetParameters=true}
  ]
  rendererClasses [&NiagaraSpriteRendererProperties, &NiagaraMeshRendererProperties]
  renderers [
   {rendererClass=&NiagaraSpriteRendererProperties}
   {rendererClass=&NiagaraMeshRendererProperties, rendererIndex=1}
   {rendererClass=&NiagaraSpriteRendererProperties, rendererIndex=2}
   {rendererClass=&NiagaraSpriteRendererProperties, rendererIndex=3}
  ]
 }
 SegmentEmbers {
  simTarget=GPUComputeSim
  emitterUpdateScript [
   &EmitterState()
   &SpawnBurst_Instantaneous(
    Spawn Count=&MakeIntFromFloat(Float=&Multiply_FloatByInt(Float=61.256, Integer=$User.NumSegments))
    Use Spawn Probability=false
    UseLoopCountLimit=false
   )
   &SpawnPerUnit(Module.Use Max Movement Threshold=false, Module.Use Movement Tolerance=false, Spawn Spacing=100, Use Spawn Probability=false) {enabled=false}
   &SpawnRate(SpawnRate=0, Use Spawn Probability=false) {enabled=false, module=SpawnRate001}
   &SpawnRate(SpawnRate=&Multiply_FloatByInt(Float=100, Integer=$User.NumSegments), Use Spawn Probability=false) {enabled=false}
  ]
  particleSpawnScript [
   &InitializeParticle(
    Color=<rgba 0.5 0.5 0.5 1>
    Lifetime=&RandomRangeFloat(Minimum=0.25, Maximum=0.85)
    Sprite Size Mode=::NewEnumerator3
    Uniform Sprite Size=&RandomRangeFloat(Minimum=5, Maximum=20)
    UsePositionOffset=false
    Write Material Random=true
   )
   &PartitionParticles(Number Of Partitions=$User.NumSegments)
   /SetVariables(Particles.OriginPos:NiagaraPosition=&TransformPosition(
     Position=&Multiply_VectorByFloat(
      Vector=&Multiply_Vector(
       A=&SelectVectorFromArray(Direct Array Index=$Particles.PartitionParticles.Partition, Vector Selection Array=$User.SegmentPositions)
       B=$User.RootScale
      )
      Float=1
     )
     Source Space=::Local
     Destination Space=::Local
    )) {isSetParameters=true}
   &ShapeLocation(
    Custom Transform Matrix:NiagaraMatrix=&MultiplyMatrix(
     Matrix A:NiagaraMatrix=&MakeTransform(
      Inverse:NiagaraBool=false
      Position:Vector3f=?
      Local Offset:Vector3f=<0 0 0>
      Scale:Vector3f=<1 1 1>
      Rotation Interface:ENiagara_RotationDefinitionApproach=::NewEnumerator1
      Axis:Vector3f=?
      Quaternion:Quat4f=&MultiplyQuaternion(
       Quaternion A=&self:ScratchDynamicInput_02(Index:NiagaraInt32=$Particles.PartitionParticles.Partition, QuatArray:NiagaraDataInterfaceArrayQuat=$User.SegmentRotations)
       Quaternion B=&MakeQuatFromAxisAngle(Axis=<1 0 0>, AngleInDegrees=90)
      )
      Rotation Matrix:NiagaraMatrix=?
      X Vector:Vector3f=?
      Yaw:NiagaraFloat=?
      Angle:NiagaraFloat=?
      Pitch:NiagaraFloat=?
      Y Vector:Vector3f=?
      Roll:NiagaraFloat=?
      Z Vector:Vector3f=?
     )
     Matrix B:NiagaraMatrix=&MakeTransform(
      Inverse:NiagaraBool=false
      Position:Vector3f=Linked() {linkedVariable {name=Particles.OriginPos, type=NiagaraPosition}}
      Local Offset:Vector3f=<0 0 0>
      Scale:Vector3f=<1 1 1>
      Rotation Interface:ENiagara_RotationDefinitionApproach=::NewEnumerator2
      Axis:Vector3f=?
      Quaternion:Quat4f=?
      Rotation Matrix:NiagaraMatrix=?
      X Vector:Vector3f=?
      Yaw:NiagaraFloat=0
      Angle:NiagaraFloat=?
      Pitch:NiagaraFloat=0
      Y Vector:Vector3f=?
      Roll:NiagaraFloat=0
      Z Vector:Vector3f=?
     )
    )
    Cylinder Height=130
    Cylinder Radius=1
    MaskSpawnGroup=false
    Override Seed=false
    Shape Primitive=::NewEnumerator1
    Surface Only=false
    Transform Method:ENiagara_TransformType=::NewEnumerator1
   ) {module=ShapeLocation001}
   &AddVelocity(
    Velocity=$Particles.ShapeLocation001.ShapeNormal
    Velocity Speed Scale=&RandomRangeFloat(Minimum=0, Maximum=0)
   ) {module=AddVelocity001}
   &KillParticles(Kill Particles=&InvertBool(Boolean:NiagaraBool=&self:ScratchDynamicInput(SegmentIndex:NiagaraInt32=$Particles.PartitionParticles.Partition, VisibilityArray:NiagaraDataInterfaceArrayBool=$User.ActiveArray)))
   &AddVelocity(
    Cone Angle=79.6511
    Cone Axis=<-1 0 0>
    Module.UseConeFalloff=false
    Module.UseVelDistribution=false
    Override Seed=false
    Velocity Mode=::NewEnumerator2
    Velocity Speed=&RandomRangeFloat(Minimum=600, Maximum=1250)
   ) {enabled=false}
   &AddVelocity(
    Cone Angle=65
    Module.UseConeFalloff=false
    Module.UseVelDistribution=false
    Override Seed=false
    Velocity Mode=::NewEnumerator2
    Velocity Speed=&RandomRangeFloat(Minimum=1000, Maximum=2100)
   ) {module=AddVelocity002}
   /SetVariables(Particles.Velocity=&Multiply_Vector(A=$Particles.Velocity, B=<1 1 0.1>)) {isSetParameters=true}
  ]
  particleUpdateScript [
   &ParticleState()
   &ScaleSpriteSize(
    Uniform Curve Index=&FN_LOD_Distance_Remap(Start Distance=1000, End Distance=5000, Minimum Output=0, Maximum Output=1, Clamp Output=true)
    Uniform Curve Scale=0.5
    Uniform Curve Sprite Scale=DataInterface() {propertyValues {Curve {keys [{interpMode=RCIM_Cubic, tangentMode=RCTM_User, time=0, value=1}, {interpMode=RCIM_Cubic, tangentMode=RCTM_User, time=1, value=2}]}}}
   )
   &Drag(Drag=&RandomRangeFloat(Minimum=3.5, Maximum=6), Use Linear Drag=true, Use Rotational Drag=true)
   /Niagara/Modules/Update/Forces/V2/CurlNoiseForce(
    Noise Frequency=10
    Noise Quality / Cost=::NewEnumerator0
    Noise Strength=&FloatFromCurve(
     FloatCurve=DataInterface() {propertyValues {Curve {keys [{tangentMode=RCTM_User, time=0, value=0}, {tangentMode=RCTM_User, time=1, value=1}]}}}
     Scale Curve=2500
    )
    Pan Noise=false
   )
   &AccelerationForce(
    Acceleration=&Multiply_VectorByFloat(Vector=<1 0 0>, Float=&RandomRangeFloat(Minimum=0, Maximum=1250))
    Coordinate Space=::Local
   )
   &SolveForcesAndVelocity(Clamp Velocity=false, Limit Acceleration=false)
   &ScaleSpriteSizeBySpeed(Max Scale Factor=<2 5>)
   &ScaleColor(
    Scale RGB=&VectorFromFloat(Value=&FloatFromCurve(
      FloatCurve=DataInterface() {propertyValues {Curve {keys [
          {arriveTangent=9.0725, interpMode=RCIM_Cubic, leaveTangent=9.0725, tangentMode=RCTM_User, time=0, value=0}
          {arriveTangent=-2.5249, interpMode=RCIM_Cubic, leaveTangent=-2.5249, tangentMode=RCTM_User, time=0.1408, value=0.9847}
          {arriveTangent=-0.0405, interpMode=RCIM_Cubic, leaveTangent=-0.0405, tangentMode=RCTM_User, time=1, value=0}
         ]}}}
      Scale Curve=5
     ))
    ScaleA=true
    ScaleRGB=true
   )
  ]
  rendererClasses [&NiagaraSpriteRendererProperties]
  renderers [{rendererClass=&NiagaraSpriteRendererProperties}]
 }
}
summary {
 systemName=NS_AnnoyedMop_Prj_OnDeath
 emitters [
  {emitterName=SegmentEmbers, simTarget=GPUComputeSim, rendererClasses [&NiagaraSpriteRendererProperties]}
  {emitterName=Linger_Fire, rendererClasses [&NiagaraSpriteRendererProperties, &NiagaraMeshRendererProperties]}
  {
   bEnabled=false
   emitterName=FireEject
   rendererClasses [&NiagaraSpriteRendererProperties, &NiagaraMeshRendererProperties]
  }
 ]
 userVariables [
  User.ActiveArray {
   type=NiagaraDataInterfaceArrayBool
   description=""
   defaultValue {
    struct=NiagaraExt_VariableValue_DataInterface
    value {
     dataInterface=&self:NiagaraDataInterfaceArrayBool_0
     dataInterfaceClass=/Script/Niagara.NiagaraDataInterfaceArrayBool
    }
   }
  }
  User.BaseThicknessMultiplier {type=NiagaraFloat, defaultValue=1, description=""}
  User.DebugShapes {type=NiagaraBool}
  User.DustLocation {type=NiagaraPosition, description="", defaultValue {}}
  User.GlobalScale {type=NiagaraFloat, defaultValue=0, description=""}
  User.GroundDust {type=NiagaraBool, defaultValue=true}
  User.HeaderTiling {type=NiagaraFloat, defaultValue=1.5, description=""}
  User.isMassive {type=NiagaraBool}
  User.NumSegments {type=NiagaraInt32, defaultValue=5}
  User.PrjSize {type=NiagaraInt32}
  User.RootScale {type=Vector3f, description="", defaultValue {struct=Vector3f, value=<1 1 1>}}
  User.ScaleMovement {type=NiagaraFloat, defaultValue=0, description=""}
  User.SegmentMesh {
   type=StaticMesh
   description=""
   defaultValue {struct=NiagaraExt_VariableValue_Object, value {object=null, objectClass=&StaticMesh}}
  }
  User.SegmentPositions {
   type=NiagaraDataInterfaceArrayFloat3
   description=""
   defaultValue {
    struct=NiagaraExt_VariableValue_DataInterface
    value {
     dataInterface=&self:NiagaraDataInterfaceArrayFloat3_0
     dataInterfaceClass=&NiagaraDataInterfaceArrayFloat3
    }
   }
  }
  User.SegmentRotations {
   type=NiagaraDataInterfaceArrayQuat
   description=""
   defaultValue {
    struct=NiagaraExt_VariableValue_DataInterface
    value {
     dataInterface=&self:NiagaraDataInterfaceArrayQuat_0
     dataInterfaceClass=/Script/Niagara.NiagaraDataInterfaceArrayQuat
    }
   }
  }
  User.SegmentWidth {type=NiagaraInt32, defaultValue=60}
  User.TargetSM {
   type=NiagaraDataInterfaceSkeletalMesh
   defaultValue {
    struct=NiagaraExt_VariableValue_DataInterface
    value {dataInterface=&self:NiagaraDataInterfaceSkeletalMesh_0}
   }
  }
  User.TrailingRibbonWidth {type=NiagaraFloat, defaultValue=250, description=""}
 ]
}
